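# Physics asset for the Powerful Dozen pickaxe skeleton, set up for a dangling five-bone tassel on the axe.
# Capsule bodies form one chain from root through offset, handle and a tiny tassel master body to the five tassel segments.
# Joints from root down to the master body leave swing and twist free.
# Tassel joints use wide swing limits with very tight twist limits, so the tassel bends freely but barely spins.
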PhysicsAsset /BRCosmetics/Weapons/FORT_Melee/Pickaxe_PowerfulDozen/Meshes/PowerfulDozen_Axe_Physics
bodies {
 dyn_tassle_01 [{shapeType=Capsule, length=5, radius=1.2, shapeName=dyn_tassle_02_capsule, center=<2.86 -1 0>, rotation=<rot 90 0 0>}]
 dyn_tassle_01_MSTR [{shapeType=Capsule, length=0.1, radius=0.1, shapeName=dyn_tassle_01_MSTR_capsule}]
 dyn_tassle_02 [{shapeType=Capsule, length=5, radius=1.2, shapeName=dyn_tassle_02_capsule, center=<3.49 0 0>, rotation=<rot 90 0 0>}]
 dyn_tassle_03 [{shapeType=Capsule, length=4.7, radius=1.2, shapeName=dyn_tassle_03_capsule, center=<3.69 0 0>, rotation=<rot 90 0 0>}]
 dyn_tassle_04 [{shapeType=Capsule, length=4.5, radius=1.3, shapeName=dyn_tassle_04_capsule, center=<3.53 0 0>, rotation=<rot 90 0 0>}]
 dyn_tassle_05 [{shapeType=Capsule, length=5, radius=1.3, shapeName=dyn_tassle_04_capsule, center=<3.87 0 0>, rotation=<rot 90 0 0>}]
 handle [{shapeType=Capsule, length=47, radius=2.5, shapeName=handle_capsule, center=<0 0 13.0242>}]
 offset [{shapeType=Capsule, length=0.505, radius=0.505, shapeName=offset_capsule, rotation=<rot 0 -90 0>}]
 root [{shapeType=Capsule, length=0.505, radius=0.505, shapeName=root_capsule, rotation=<rot 0 -90 0>}]
}
constraints [
 {
  bone1Name=offset
  bone2Name=root
  swing1LimitDegrees=45
  swing1Motion=Free
  swing2LimitDegrees=45
  swing2Motion=Free
  twistLimitDegrees=45
  twistMotion=Free
 }
 {
  bone1Name=handle
  bone2Name=offset
  swing1LimitDegrees=45
  swing1Motion=Free
  swing2LimitDegrees=45
  swing2Motion=Free
  twistLimitDegrees=45
  twistMotion=Free
 }
 {bone1Name=dyn_tassle_03, bone2Name=dyn_tassle_02, swing1LimitDegrees=60, swing2LimitDegrees=55, twistLimitDegrees=2.5}
 {bone1Name=dyn_tassle_04, bone2Name=dyn_tassle_03, swing1LimitDegrees=62, swing2LimitDegrees=58, twistLimitDegrees=2.5}
 {bone1Name=dyn_tassle_02, bone2Name=dyn_tassle_01, swing1LimitDegrees=65, swing2LimitDegrees=60, twistLimitDegrees=2.5}
 {
  bone1Name=dyn_tassle_01_MSTR
  bone2Name=handle
  swing1LimitDegrees=45
  swing1Motion=Free
  swing2LimitDegrees=45
  swing2Motion=Free
  twistLimitDegrees=45
  twistMotion=Free
 }
 {bone1Name=dyn_tassle_01, bone2Name=dyn_tassle_01_MSTR, swing1LimitDegrees=77, swing2LimitDegrees=77, twistLimitDegrees=2}
 {bone1Name=dyn_tassle_05, bone2Name=dyn_tassle_04, swing1LimitDegrees=72, swing2LimitDegrees=68, twistLimitDegrees=2.5}
]
