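# Animation Blueprint for the PowerfulDozen pickaxe that adds physics-driven secondary motion to the axe mesh.
# The cached input pose feeds two RigidBody nodes, both simulated in the space of the handle bone with different physics assets.
# A boolean blend on bIsInFrontEnd selects between the front-end simulation, which overrides world gravity, and the other, which scales component velocity and acceleration per axis.
# The EventGraph is empty.
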AnimBlueprint /BRCosmetics/Weapons/FORT_Melee/Pickaxe_PowerfulDozen/Meshes/PowerfulDozen_Axe_AnimBP
target_skeleton=/Game/Weapons/FORT_Melee/Pickaxe_PowerfulDozen/Meshes/PowerfulDozen_Axe_Skeleton
graphs [
 AnimGraph {nodes [
   n1 = RigidBody(ComponentPose=Local To Component(LocalPose=Use cached pose 'ref pose'())) {
    bClampLinearTranslationLimitToRefPose=true
    overridePhysicsAsset=&/PowerfulDozen_Axe_FrontEnd_Physics
    simulationSpace=BaseBoneSpace
    alphaBoolBlend {blendOption=Linear}
    baseBoneRef {boneName=handle}
    componentAppliedLinearAccClamp=<5000 5000 5000>
    overrideWorldGravity {z=-980}
   }
   n2 = RigidBody(ComponentPose=Local To Component(LocalPose=Use cached pose 'ref pose'())) {
    bClampLinearTranslationLimitToRefPose=true
    overridePhysicsAsset=&/PowerfulDozen_Axe_Physics
    simulationSpace=BaseBoneSpace
    alphaBoolBlend {blendOption=Linear}
    baseBoneRef {boneName=handle}
    componentAppliedLinearAccClamp=<5000 5000 5000>
    componentLinearAccScale=<0.8 0.6 0.85>
    componentLinearVelScale=<0.8 0.6 0.85>
   }
   Output Pose(Result=Blend Poses by bool(BlendPose_0=Component To Local(ComponentPose=@n1), BlendPose_1=Component To Local(ComponentPose=@n2), bActiveValue=Get bIsInFrontEnd()) {blendType=Linear}) {layerGroup=None}
   Save cached pose 'ref pose'(Pose=Input Pose()) {cacheName=ref pose}
  ]}
 EventGraph {links [], nodes []}
]
variables []
